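# Level Sequence for the Pilgrim camera manager that provides a static shot, named for the stage.
# A spawnable Cine Camera Actor takes its framing from a transform track whose values do not animate.
# Its camera component tracks set a fixed focal length and aperture.
# A single camera cut section spans the shot on a 30 fps timeline.
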LevelSequence /PilgrimCore/CameraManager/Sequences/StaticShots/LS_PilgrimCam_SS_Stage
bindings [
 CameraComponent {
  id=#1
  parent_id=#2
  possessed_class=CineCameraComponent
  tracks [
   {
    type=MovieSceneFloatTrack
    display_name=Current Aperture
    sections [{type=MovieSceneFloatSection, channels [None {type=MovieSceneScriptingFloatChannel, default=0.13}]}]
   }
   {
    type=MovieSceneFloatTrack
    display_name=Current Focal Length
    sections [{type=MovieSceneFloatSection, channels [None {type=MovieSceneScriptingFloatChannel, default=69.7271}]}]
   }
  ]
 }
 Cine Camera Actor {
  id=#2
  possessed_class=CineCameraActor
  tracks [
   {
    type=MovieScene3DTransformTrack
    sections [{
      type=MovieScene3DTransformSection
      channels [
       Location.X {type=MovieSceneScriptingDoubleChannel, default=2300.3801, const [3131.4153,1]}
       Location.Y {type=MovieSceneScriptingDoubleChannel, default=-532.3064, const [-468.5136,1]}
       Location.Z {type=MovieSceneScriptingDoubleChannel, default=910.6928, const [910.6929,1]}
       Rotation.X {type=MovieSceneScriptingDoubleChannel, default=0, const [0,1]}
       Rotation.Y {type=MovieSceneScriptingDoubleChannel, default=12.0764, const [9.4087,1]}
       Rotation.Z {type=MovieSceneScriptingDoubleChannel, default=180, const [180,1]}
       Scale.X {type=MovieSceneScriptingDoubleChannel, default=1, const [1,1]}
       Scale.Y {type=MovieSceneScriptingDoubleChannel, default=1, const [1,1]}
       Scale.Z {type=MovieSceneScriptingDoubleChannel, default=1, const [1,1]}
      ]
     }]
   }
   {type=MovieSceneSpawnTrack, sections [{type=MovieSceneSpawnSection, channels [None {type=MovieSceneScriptingBoolChannel, default=true}]}]}
  ]
 }
]
display_rate {denominator=1, numerator=30}
tick_resolution {denominator=1, numerator=24000}
tracks [{type=MovieSceneCameraCutTrack, sections [{type=MovieSceneCameraCutSection, end_frame=300, start_frame=0}]}]
